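# Translucent, unlit, two-sided Niagara ribbon and mesh material for a trail in the LazarusLens skydiving contrail.
# Noise-modulated wave-UV bands, a dissolve texture and turbulence-distorted cell and cloud noise combine into a masked wispy flame shape.
# That shape indexes a curve atlas for time-of-day-scaled emissive and, with particle alpha, forms opacity, depth-faded on deferred paths.
# Dynamic parameters drive wave timing, amplitude, panning and sine-based vertex offset strength.
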Material /FlamingHoopsGameplay/Athena/Effects/Particles/Skydiving/NS_Contrail_LazarusLens/FX/M_Contrail_LazarusLens_Trail2
graph {
 properties {
  bUsedWithNiagaraMeshParticles=true
  bUsedWithNiagaraRibbons=true
  blendMode=BLEND_Translucent
  pixelDepthOffsetMode=PDOM_Legacy
  shadingModel=MSM_Unlit
  twoSided=true
  usedShadingModels=Unlit
 }
 nodes [
  n1 = TextureCoordinate()
  n2 = ComponentMask(None=@n1) {g=true}
  n3 = ComponentMask(None=@n1) {r=true}
  n4 = OneMinus(None=@n3)
  n5 = Saturate(None=Multiply(A=OneMinus(None=@n2), B=Multiply(A=@n3, B=@n4)))
  n6 = Constant3Vector() {constant=<rgba 1 0 0 0>}
  n7 = Add(A=@n3, B=Multiply(A=Constant() {r=0.1}, B=Time()))
  n8 = Sine(None=@n7) {period=0.3}
  n9 = Constant3Vector() {constant=<rgba 0 1 0 0>}
  n10 = Sine(None=@n7) {period=0.3}
  n11 = Add(
   A=Transform(None=Multiply(A=@n6, B=@n8)) {transformSourceType=TRANSFORMSOURCE_Local}
   B=Transform(None=Multiply(A=@n9, B=@n10)) {transformSourceType=TRANSFORMSOURCE_Local}
  )
  n12 = Constant() {r=35}
  n13 = DynamicParameter() {parameterIndex=3, paramNames [WPO,Param2,Param3,Param4]}
  Add(B=Multiply(A=Multiply(A=@n5, B=@n11), B=LinearInterpolate(B=@n12, Alpha=@n13.WPO) {constA=20}) {constB=10}) {constB=0}
  ComponentMask(None=TextureCoordinate() {uTiling=4}) {g=true}
  n14 = DynamicParameter() {defaultValue=<rgba 2 1 1 1>, paramNames [ScaleByCenter, SineFrequency, SineStrength, SineTimer]}
  n15 = Constant()
  n16 = Panner(
   Coordinate=TextureCoordinate()
   Speed=AppendVector(
    A=ScalarParameter() {defaultValue=-0.01, group=UV Distortion, parameterName=UV Speed X 2, sliderMax=5, sliderMin=-5}
    B=ScalarParameter() {defaultValue=0.1, group=UV Distortion, parameterName=UV Speed Y 2, sliderMax=5, sliderMin=-5}
   )
  ) {speedX=-0.5}
  n17 = AppendVector(
   A=ScalarParameter() {defaultValue=1, group=UV Distortion, parameterName=UV Scale X 2, sliderMax=5}
   B=ScalarParameter() {defaultValue=0.6, group=UV Distortion, parameterName=UV Scale Y 2, sliderMax=5}
  )
  n18 = TextureSampleParameter2D(UVs=Add(A=AppendVector(A=@n14.ScaleByCenter, B=@n15), B=Multiply(A=@n16, B=@n17))) {
   group=UV Distortion
   parameterName=Distortion Texture
   samplerType=SAMPLERTYPE_LinearColor
   texture=/Game/Effects/Fort_Effects/Textures/Plasma/T_Turbulence_AdditiveWarp_Linear
  }
  n19 = DynamicParameter() {defaultValue=<rgba 2 1 1 1>, paramNames [ScaleByCenter, SineFrequency, SineStrength, SineTimer]}
  n20 = Constant()
  n21 = Panner(
   Coordinate=TextureCoordinate()
   Speed=AppendVector(
    A=ScalarParameter() {defaultValue=0.05, group=UV Distortion, parameterName=UV Speed X, sliderMax=5, sliderMin=-5}
    B=ScalarParameter() {defaultValue=0.2, group=UV Distortion, parameterName=UV Speed Y, sliderMax=5, sliderMin=-5}
   )
  ) {speedX=-0.5}
  n22 = AppendVector(
   A=ScalarParameter() {defaultValue=1, group=UV Distortion, parameterName=UV Scale X, sliderMax=5}
   B=ScalarParameter() {defaultValue=0.5, group=UV Distortion, parameterName=UV Scale Y, sliderMax=5}
  )
  n23 = TextureSampleParameter2D(UVs=Add(A=AppendVector(A=@n19.ScaleByCenter, B=@n20), B=Multiply(A=@n21, B=@n22))) {
   group=UV Distortion
   parameterName=Distortion Texture
   samplerType=SAMPLERTYPE_LinearColor
   texture=&T_Turbulence_AdditiveWarp_Linear
  }
  n24 = Subtract(A=MaterialFunctionCall(X=Multiply(A=@n18.R, B=@n23.R), Y=Multiply(A=@n18.G, B=@n23.G)) {materialFunction=&MakeFloat2}.Result) {constB=0}
  ConstantBiasScale(None=@n24) {bias=-0.5, scale=2}
  n25 = TextureCoordinate()
  n26 = Multiply(
   A=NamedRerouteUsage()
   B=ScalarParameter() {defaultValue=0.2, group=Wispy Flame, parameterName=Wispy Distortion Strength, sliderMax=1, sliderMin=-1}
  ) {constB=0.1}
  n27 = ScalarParameter() {group=Wispy Flame, parameterName=Speed X, sliderMax=5, sliderMin=-5}
  n28 = ScalarParameter() {defaultValue=0.2, group=Wispy Flame, parameterName=Speed Y 2, sliderMax=5, sliderMin=-5}
  n29 = Multiply(
   A=Panner(Coordinate=Add(A=@n25, B=@n26), Speed=AppendVector(A=@n27, B=@n28)) {speedX=1}
   B=AppendVector(
    A=ScalarParameter() {defaultValue=1, group=Wispy Flame, parameterName=Scale X 2, sliderMax=5}
    B=ScalarParameter() {defaultValue=1, group=Wispy Flame, parameterName=Scale Y 2, sliderMax=5}
   )
  )
  n30 = TextureCoordinate()
  n31 = SmoothStep(Value=Power(Base=ComponentMask(None=@n30) {g=true})) {constMax=0.6}
  n32 = Power(Base=OneMinus(None=@n31)) {constExponent=0.7}
  n33 = Multiply(
   A=SmoothStep(Value=TextureSample(UVs=@n29) {samplerType=SAMPLERTYPE_LinearGrayscale, texture=/Game/Effects/Fort_Effects/Textures/Tile/T_CellNoise_Tile_M}.R) {constMax=0.5, constMin=0.25}
   B=@n32
  )
  n34 = NamedRerouteUsage()
  n35 = Constant() {r=2}
  n36 = NamedRerouteUsage()
  n37 = Constant()
  n38 = Multiply(A=Time()) {constB=0.3}
  n39 = Subtract(A=ComponentMask(None=MaterialFunctionCall!(UV=@n34, U Frequency, V Frequency=@n35, U Amplitude, V Amplitude=@n36, U Speed, V Speed, U Time=@n37, V Time=@n38) {materialFunction=/Game/Effects/Fort_Effects/MaterialFunctions/MF_WaveUV}.Wave UVs) {r=true}) {constB=0.6, desc=s}
  n40 = Saturate(None=Divide(A=Multiply(A=@n39) {constB=2}) {constB=0.1})
  n41 = NamedRerouteUsage()
  n42 = Constant() {r=3}
  n43 = Multiply(A=NamedRerouteUsage()) {constB=0.5}
  n44 = Subtract(A=ComponentMask(None=MaterialFunctionCall(UV=@n41, V Frequency=@n42, V Amplitude=@n43, U Time=@n37, V Time=@n37) {materialFunction=&MF_WaveUV}.Wave UVs) {r=true}) {constB=0.6, desc=s}
  n45 = Saturate(None=Divide(A=Multiply(A=@n44) {constB=2}, B=Constant() {r=0.1}))
  n46 = MaterialFunctionCall(Float2=TextureCoordinate()) {materialFunction=&BreakOutFloat2Components}
  n47 = MaterialFunctionCall(X=@n46.G, Y=@n46.G) {materialFunction=&MakeFloat2}
  n48 = MaterialFunctionCall(
   UVs=@n47.Result
   Texture Scale=DynamicParameter() {defaultValue=<rgba 2 1 1 1>, paramNames [ScaleByCenter, SineFrequency, SineStrength, SineTimer]}.ScaleByCenter
  ) {materialFunction=&ScaleUVsByCenter}
  n49 = NamedRerouteUsage()
  n50 = Constant()
  n51 = MaterialFunctionCall(UV=NamedRerouteUsage(), V Frequency=Constant() {r=4}, U Amplitude=@n49, V Amplitude=Multiply(A=@n49), U Time=@n50, V Time=@n50) {materialFunction=&MF_WaveUV}
  n52 = Multiply(A=Subtract(A=ComponentMask(None=@n51.Wave UVs) {r=true}) {constB=0.5, desc=s}) {constB=2}
  n53 = Add(
   A=MaterialFunctionCall(
    X=MaterialFunctionCall(Float2=@n48.UVs) {materialFunction=&BreakOutFloat2Components}.R
    Y=MaterialFunctionCall(Float2=@n47.Result) {materialFunction=&BreakOutFloat2Components}.G
   ) {materialFunction=&MakeFloat2}.Result
   B=MaterialFunctionCall(X=Divide(A=@n52) {constB=1}, Y=Constant()) {materialFunction=&MakeFloat2}.Result
  )
  n54 = VectorParameter() {parameterName=Uv, defaultValue=<rgba 1 1 1 -1>}
  n55 = AppendVector(A=@n54.R, B=@n54.R)
  n56 = DynamicParameter() {parameterIndex=2, defaultValue=<rgba 0 0.2 1 1>, paramNames [SpeedX,SpeedY,Param3,Param4]}
  n57 = Panner(Coordinate=TextureCoordinate(), Speed=AppendVector(A=@n56.SpeedX, B=@n56.SpeedX)) {speedY=0.2}
  n58 = AppendVector(A=@n54.B, B=@n54.B)
  n59 = TextureSample(UVs=Add(A=Multiply(A=@n53, B=@n55), B=Multiply(A=@n57, B=@n58))) {texture=/Game/CharClasses/Outlander/Gadgets/AntiMaterialCharge/Materials/T_AntiMaterialCharge_Dissolve_01}
  n60 = Subtract(A=Saturate(None=Add(A=@n40, B=@n45)), B=SmoothStep(Value=@n59.R) {constMin=0.5})
  n61 = OneMinus(None=@n59.R)
  n62 = Multiply(A=Multiply(A=SmoothStep(Value=@n61) {constMin=0.2}, B=@n31), B=@n60)
  n63 = Add(A=@n33, B=Subtract(A=@n60, B=Multiply(A=@n62) {constB=0.5}))
  n64 = Add(
   A=TextureCoordinate()
   B=Multiply(
    A=NamedRerouteUsage()
    B=ScalarParameter() {defaultValue=0.2, group=Wispy Flame, parameterName=Wispy Distortion Strength, sliderMax=1, sliderMin=-1}
   ) {constB=0.1}
  )
  n65 = AppendVector(
   A=ScalarParameter() {group=Wispy Flame, parameterName=Speed X, sliderMax=5, sliderMin=-5}
   B=ScalarParameter() {defaultValue=0.3, group=Wispy Flame, parameterName=Speed Y, sliderMax=5, sliderMin=-5}
  )
  n66 = ScalarParameter() {defaultValue=1, group=Wispy Flame, parameterName=Scale X, sliderMax=5}
  n67 = ScalarParameter() {defaultValue=0.5, group=Wispy Flame, parameterName=Scale Y, sliderMax=5}
  n68 = TextureSampleParameter2D(UVs=Multiply(A=Panner(Coordinate=@n64, Speed=@n65) {speedX=1}, B=AppendVector(A=@n66, B=@n67))) {
   group=Wispy Flame
   parameterName=Flame Texture
   samplerType=SAMPLERTYPE_LinearGrayscale
   texture=/Game/Effects/Fort_Effects/Emotes/UltraLight/Texture/T_Cloud_Greyscale
  }
  n69 = Multiply(A=SmoothStep(Value=@n68.R) {constMax=0.6, constMin=0.4}, B=@n32)
  n70 = SmoothStep(Value=Subtract(A=Saturate(None=@n63), B=Saturate(None=@n69))) {constMin=0.1}
  n71 = ScalarParameter() {defaultValue=0.5, group=Wispy Flame, parameterName=Candle Shape Mask, sliderMax=5}
  n72 = Multiply(A=ComponentMask(None=TextureCoordinate()) {g=true})
  n73 = Multiply(A=Multiply(A=Saturate(None=@n70), B=Multiply(A=@n71, B=@n72))) {constB=3}
  n74 = LinearInterpolate(A=@n73, B=Power(Base=@n72) {constExponent=10})
  n75 = SmoothStep(
   Min=ScalarParameter() {group=Color, parameterName=Color Smooth Step Min, sliderMax=1, sliderMin=-1}
   Max=ScalarParameter() {defaultValue=2, group=Color, parameterName=Color Smooth Step Max, sliderMax=5}
   Value=Power(Base=Saturate(None=@n74), Exp=ScalarParameter() {defaultValue=0.2, group=Color, parameterName=Color Power, sliderMax=5}) {constExponent=1}
  ) {constMax=2}
  n76 = CurveAtlasRowParameter(CurveTime=@n75) {
   atlas=/Game/Athena/LinearColorCurves/CurveAtlas_Chainmail
   curve=/Game/Athena/LinearColorCurves/Curve_TealBlack
   defaultValue=4
   parameterName=Curve_Color
  }
  n77 = Add(A=Multiply(A=@n75, B=ScalarParameter() {defaultValue=1, parameterName=Emissive, sliderMax=2})) {constB=0.5}
  n78 = VectorParameter() {parameterName=TOD color, defaultValue=<rgba 1 1 1 0.5>}
  MaterialFunctionCall(Emissive=Multiply(
    A=Multiply(A=@n76, B=@n77)
    B=MaterialFunctionCall(Value_Morning=@n78.R, Value_Day=@n78.R, Value_Evening=@n78.R, Value_Night=@n78.R) {materialFunction=&Scalar_Time_of_Day}.Result
   )) {materialFunction=&MF_FX_EmissiveAdjustments}
  Multiply(A=@n33)
  n79 = ScalarParameter() {defaultValue=0.9, group=Shape Lerp, parameterName=Shape Lerp Min, sliderMax=2, sliderMin=-2}
  n80 = ScalarParameter() {defaultValue=1, group=Shape Lerp, parameterName=Shape Lerp Max, sliderMax=2, sliderMin=-2}
  n81 = Power(Base=@n72, Exp=ScalarParameter() {defaultValue=2, group=Shape Lerp, parameterName=Shape Lerp Power, sliderMax=10}) {constExponent=1}
  Multiply(A=@n73, B=OneMinus(None=SmoothStep(Min=@n79, Max=@n80, Value=@n81)))
  n82 = ScalarParameter() {defaultValue=8, parameterName=WPOStrenght}
  n83 = DynamicParameter() {parameterIndex=3, paramNames [WPO,Param2,Param3,Param4]}
  n84 = Constant() {r=0.5}
  n85 = Multiply(A=SphereMask(A=TextureCoordinate(), B=@n84, Radius=@n84, Hardness=Constant())) {constB=1.5}
  n86 = Multiply(A=Time(), B=ScalarParameter() {defaultValue=2, parameterName=TimeMult}) {constB=2}
  Multiply(
   A=Multiply(A=Constant3Vector() {constant=<rgba 1 0 0 0>}, B=LinearInterpolate(A=@n82, Alpha=@n83.WPO) {constA=5, constB=0}) {constB=5}
   B=Sine(None=Add(A=@n85, B=@n86))
  )
  NamedRerouteDeclaration(None=Subtract(A=@n24) {constB=0.1}) {name=Distortion, nodeColor=<rgba 1 0.3281 0 1>}
  n87 = OneMinus(None=Time())
  n88 = DynamicParameter() {defaultValue=<rgba 2 1 1 1>, paramNames [ScaleByCenter, SineFrequency, SineStrength, SineTimer]}
  NamedRerouteDeclaration(None=Add(A=Multiply(A=@n87, B=@n88.SineStrength) {constB=2}, B=@n88.SineTimer)) {name=Time, nodeColor=<rgba 1 0.0469 0 1>}
  n89 = NamedRerouteUsage()
  n90 = DynamicParameter() {parameterIndex=1, defaultValue=<rgba 0.25 0.2 0.25 1>, paramNames [AmplitudeNoiseSpeed, AmplitudeContrast, AmplitudeStrenght, TrailTimeAdd]}
  n91 = Divide(A=@n90.AmplitudeNoiseSpeed) {constB=4}
  n92 = TextureSample(UVs=Panner(
    Coordinate=TextureCoordinate() {uTiling=0.25, vTiling=0}
    Time=Multiply(A=@n89) {constB=-0.2}
    Speed=AppendVector(A=@n90.AmplitudeNoiseSpeed, B=@n91)
   ) {speedX=0.25}) {texture=/Game/Effects/Fort_Effects/Textures/Tile/T_GaussianNoise_01_M}
  NamedRerouteDeclaration(None=Multiply(A=Add(A=@n92.RGB, B=@n90.AmplitudeContrast) {constB=0.5}, B=@n90.AmplitudeStrenght)) {name=AmplitudeR, nodeColor=<rgba 0.9688 0 1 1>}
  n93 = NamedRerouteUsage()
  n94 = DynamicParameter() {parameterIndex=1, defaultValue=<rgba 0.25 0.2 0.25 1>, paramNames [AmplitudeNoiseSpeed, AmplitudeContrast, AmplitudeStrenght, TrailTimeAdd]}
  n95 = Divide(A=@n94.AmplitudeNoiseSpeed) {constB=4}
  n96 = TextureSample(UVs=Panner(
    Coordinate=TextureCoordinate() {uTiling=0.25, vTiling=0}
    Time=Multiply(A=@n93) {constB=0.2}
    Speed=AppendVector(A=@n94.AmplitudeNoiseSpeed, B=@n95)
   ) {speedX=0.25}) {texture=&T_GaussianNoise_01_M}
  NamedRerouteDeclaration(None=Multiply(A=Add(A=@n96.RGB, B=@n94.AmplitudeContrast) {constB=0.5}, B=@n94.AmplitudeStrenght)) {name=AmplitudeL, nodeColor=<rgba 0.9688 0 1 1>}
  n97 = MaterialFunctionCall(Float2=TextureCoordinate()) {materialFunction=&BreakOutFloat2Components}
  NamedRerouteDeclaration(None=MaterialFunctionCall(X=@n97.G, Y=@n97.G) {materialFunction=&MakeFloat2}.Result) {name=UV, nodeColor=<rgba 0.2891 1 0 1>}
  ParticleColor()
  n98 = ScalarParameter() {defaultValue=0.1, parameterName=Edge Hardness, sliderMax=1}
  n99 = Multiply(A=Multiply(A=@n74, B=ParticleColor().A), B=SmoothStep(Max=OneMinus(None=@n98), Value=@n74) {constMax=0.1, constMin=0.01}) {constA=1}
  n100 = DepthFade(Opacity=@n99, FadeDistance=ScalarParameter() {defaultValue=35, parameterName=Depth Fade Distance, sliderMax=100}) {fadeDistanceDefault=25}
  ShadingPathSwitch(Default=@n100, Deferred=@n100, Forward=@n99, Mobile=@n99)
 ]
}
